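# Material function for Apollo ground foliage that computes world-space grass color, masks, normals, specular and roughness, with separate terrain spec and roughness outputs.
# One grayscale texture, sampled at several world XY scales and an offset, blends four grass colors and supplies detail variation.
# A world mask texture blends in dirt-edge colors, and detail adjusts color by camera distance.
# Normals blend a world-aligned noise map toward the vertex normal over distance.
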MaterialFunction /Game/Athena/Apollo/Environments/GroundFoliage/MF_ApolloGrassMasks
graph {nodes [
  Constant3Vector() {constant=<rgba 0.1309 0.17 0.0338 1>}
  Constant3Vector() {constant=<rgba 0.2 0.1163 0.0481 1>}
  Constant3Vector() {constant=<rgba 0.22 0.1622 0.0632 1>}
  Constant() {r=1.25}
  Constant() {r=-4500}
  Constant() {r=1}
  Constant() {r=7000}
  Constant() {r=0.4}
  Constant() {r=1}
  Constant() {r=500}
  Constant()
  Constant() {r=1000}
  Constant() {r=-4500}
  Constant()
  Constant() {r=0.3}
  Constant() {r=0.8}
  Constant() {r=1}
  Constant()
  Constant() {r=0.3}
  Constant() {r=0.8}
  Constant() {r=7000}
  Constant() {r=1}
  Constant() {r=1.5}
  Constant() {r=-5}
  Constant()
  Constant() {r=1.5}
  Constant() {r=-1.5}
  n1 = VectorParameter() {parameterName=GrassColor01, defaultValue=<rgba 0.1021 0.16 0.0124 1>}
  n2 = VectorParameter() {parameterName=GrassColor02, defaultValue=<rgba 0.0557 0.11 0.0102 1>}
  n3 = ScalarParameter() {defaultValue=1.25, parameterName=GrassColorMask-SizeMod}
  n4 = ComponentMask(None=WorldPosition() {worldPositionShaderOffset=WPT_ExcludeAllShaderOffsets}.XYZ) {g=true, r=true}
  n5 = ScalarParameter() {defaultValue=7000, parameterName=GrassColorMask-Size}
  n6 = TextureSample(UVs=Multiply(A=@n3, B=Divide(A=@n4, B=@n5))) {
   samplerSource=SSM_Wrap_WorldGroupSettings
   samplerType=SAMPLERTYPE_Grayscale
   texture=/Game/Athena/Apollo/Environments/Landscape/Materials/Biome/CommonMaterials/T_ApolloGrass_ColorMask
  }
  n7 = Saturate(None=LinearInterpolate(A=ScalarParameter() {parameterName=L1A}, B=ScalarParameter() {defaultValue=1.5, parameterName=L1B}, Alpha=@n6.R))
  n8 = ComponentMask(None=WorldPosition() {worldPositionShaderOffset=WPT_ExcludeAllShaderOffsets}.XYZ) {g=true, r=true}
  n9 = ScalarParameter() {defaultValue=-4500, parameterName=GrassColorMask2-Size}
  n10 = TextureSample(UVs=Multiply(A=@n3, B=Divide(A=@n8, B=@n9))) {
   samplerSource=SSM_Wrap_WorldGroupSettings
   samplerType=SAMPLERTYPE_Grayscale
   texture=&T_ApolloGrass_ColorMask
  }
  n11 = LinearInterpolate(A=ScalarParameter() {defaultValue=1.5, parameterName=L1A2}, B=ScalarParameter() {defaultValue=-9, parameterName=L1B2}, Alpha=@n10.R)
  n12 = Add(
   A=ComponentMask(None=WorldPosition() {worldPositionShaderOffset=WPT_ExcludeAllShaderOffsets}.XYZ) {g=true, r=true}
   B=ScalarParameter() {defaultValue=500, parameterName=GrassColorMask2-Offset}
  )
  n13 = ScalarParameter() {defaultValue=-4500, parameterName=GrassColorMask2-Size}
  n14 = TextureSample(UVs=Multiply(A=@n3, B=Divide(A=@n12, B=@n13))) {
   samplerSource=SSM_Wrap_WorldGroupSettings
   samplerType=SAMPLERTYPE_Grayscale
   texture=&T_ApolloGrass_ColorMask
  }
  n15 = Saturate(None=LinearInterpolate(A=ScalarParameter() {defaultValue=-1, parameterName=L1A3}, B=ScalarParameter() {defaultValue=2, parameterName=L1B3}, Alpha=@n14.R))
  n16 = LinearInterpolate(
   A=LinearInterpolate(
    A=LinearInterpolate(A=@n1.RGB, B=@n2.RGB, Alpha=@n7)
    B=VectorParameter() {parameterName=GrassColor03, defaultValue=<rgba 0.1142 0.17 0.0142 1>}.RGB
    Alpha=Saturate(None=@n11)
   )
   B=VectorParameter() {parameterName=GrassColor04, defaultValue=<rgba 0.1412 0.15 0.0305 1>}.RGB
   Alpha=@n15
  )
  n17 = LinearInterpolate(
   A=VectorParameter() {parameterName=GrassMaskDirtEdgeColor01, defaultValue=<rgba 0.2 0.1163 0.0481 1>}.RGB
   B=VectorParameter() {parameterName=GrassMaskDirtEdgeColor02, defaultValue=<rgba 0.22 0.1622 0.0632 1>}.RGB
   Alpha=@n7
  )
  n18 = MaterialFunctionCall(
   UVs=ComponentMask(None=WorldPosition() {worldPositionShaderOffset=WPT_ExcludeAllShaderOffsets}.XYZ) {g=true, r=true}
   Texture Scale=ScalarParameter() {defaultValue=11202, parameterName=GrassWorldMask}
  ) {materialFunction=&ScaleUVsByCenter}
  n19 = LinearInterpolate(
   B=Power(
    Base=TextureSample(UVs=@n18.UVs) {samplerType=SAMPLERTYPE_Masks, texture=&/T_ApolloGrass_WorldMask}.R
    Exp=ScalarParameter() {defaultValue=0.8, parameterName=GrassMaskDirtmaskPower}
   )
   Alpha=ScalarParameter() {defaultValue=1, parameterName=DEBUG-UseTerrainLayerMask}
  )
  n20 = LinearInterpolate(A=@n16, B=@n17, Alpha=@n19)
  n21 = ComponentMask(None=WorldPosition() {worldPositionShaderOffset=WPT_ExcludeAllShaderOffsets}.XYZ) {g=true, r=true}
  n22 = ScalarParameter() {defaultValue=300, parameterName=GrassDetailTexture-Size}
  n23 = TextureSample(UVs=Multiply(A=@n3, B=Divide(A=@n21, B=@n22))) {
   samplerSource=SSM_Wrap_WorldGroupSettings
   samplerType=SAMPLERTYPE_Grayscale
   texture=&T_ApolloGrass_ColorMask
  }
  n24 = Subtract(
   A=@n20
   B=LinearInterpolate(
    A=ScalarParameter() {defaultValue=0.63, parameterName=2GrsDetLerpA}
    B=ScalarParameter() {defaultValue=0.5, parameterName=2GrsDetLerpB}
    Alpha=@n23.R
   )
  )
  n25 = ScalarParameter() {defaultValue=0.5, parameterName=GGG-zzfkADD}
  n26 = Saturate(None=Add(A=Add(A=@n24, B=@n25), B=@n20))
  n27 = ScalarParameter() {defaultValue=0.5, parameterName=zzzzB-amt2}
  n28 = CameraPositionWS()
  n29 = WorldPosition()
  n30 = ScalarParameter() {defaultValue=2000, parameterName=Terrain-GrsSpecRad}
  n31 = Constant()
  FunctionOutput(None=LinearInterpolate(A=@n20, B=LinearInterpolate(A=@n26, B=@n20, Alpha=@n27), Alpha=SphereMask(A=@n28, B=@n29.XYZ, Radius=@n30, Hardness=@n31)))
  FunctionOutput(None=@n19) {outputName=WorldMask-Grass}
  n32 = ScalarParameter() {defaultValue=0.3, parameterName=TerrainLayerMaskAffect}
  FunctionOutput(None=LinearInterpolate(A=@n16, B=@n17, Alpha=Multiply(A=@n32, B=@n19))) {outputName=Result-NoDetailMask}
  FunctionOutput(None=LinearInterpolate(
    A=ScalarParameter() {defaultValue=0.4, parameterName=GrsDetLerpA}
    B=ScalarParameter() {defaultValue=1, parameterName=GrsDetLerpB}
    Alpha=@n23.R
   )) {outputName=Detail}
  FunctionOutput(None=@n7) {outputName=Red}
  FunctionOutput(None=@n14.R) {outputName=Green}
  n33 = ScalarParameter() {defaultValue=1600, group=Grass, parameterName=GrassBigTiling}
  n34 = TextureObjectParameter() {
   group=Grass
   parameterName=GrassBigNormal
   samplerSource=SSM_Wrap_WorldGroupSettings
   samplerType=SAMPLERTYPE_Normal
   texture=/Game/VisualThreatManager/StormVisuals/Textures/TilingNoise06_N
  }
  n35 = StaticBool() {value=true}
  n36 = WorldPosition() {worldPositionShaderOffset=WPT_ExcludeAllShaderOffsets}
  n37 = ScalarParameter() {defaultValue=0.55, group=Grass, parameterName=GrassBigNormalAmt}
  n38 = ScalarParameter() {defaultValue=1, group=Grass, parameterName=GrassBigNormalAmtFar}
  n39 = MaterialFunctionCall!(Blend Range=ScalarParameter() {defaultValue=15000, parameterName=DistBlndRngGrss}, Start Offset=Constant()) {materialFunction=/Engine/Functions/Engine_MaterialFunctions03/Particles/Distance_Blend}
  FunctionOutput(None=LinearInterpolate(
    A=MaterialFunctionCall!(
     Diffuse Texture Object
     Spec Texture Object
     TextureSize=@n33
     --------------------------
     Use seperate settings for Z projection
     Use seperate texture for Z projection
     Texture Size Z projection
     Z Plane Diffuse Texture Object
     Z Plane Spec Texture Object
     "--------------------------      "
     ProjectionTransitionContrast
     Use Texture Alpha Heightmap to Refine the Transistion
     HeightLerpTransitionContrast
     PickTheDiffuseHeightChannel
     TheDiffuseHeightChannel
     "--------------------------       "
     NormalTextureObject=@n34
     World Space Normals=@n35
     "--------------------------  "
     Export Float 4 Diffuse
     Export Float 4 Spec
     "--------------------------     "
     WorldPosition=@n36.XYZ
     World Space Normal
    ) {materialFunction=/Engine/Functions/Engine_MaterialFunctions02/Texturing/WorldAlignedTextures_Complex}.Normal Z Texture
    B=VertexNormalWS()
    Alpha=LinearInterpolate(A=@n37, B=@n38, Alpha=@n39.Result)
   )) {outputName=Normals}
  n40 = Multiply(A=@n23.R, B=OneMinus(None=@n15))
  FunctionOutput(None=LinearInterpolate(A=ScalarParameter() {parameterName=GrsDetSpecA}, B=ScalarParameter() {defaultValue=0.2, parameterName=GrsDetSpecB}, Alpha=@n40)) {outputName=DetailSpec}
  FunctionOutput(None=LinearInterpolate(
    A=ScalarParameter() {defaultValue=0.6, parameterName=GrsDetRoughA}
    B=ScalarParameter() {defaultValue=1, parameterName=GrsDetRoughB}
    Alpha=@n40
   )) {outputName=DetailRough}
  FunctionOutput(None=LinearInterpolate(
    A=ScalarParameter() {parameterName=Terrain-GrsDetSpecA}
    B=ScalarParameter() {defaultValue=0.3, parameterName=Terrain-GrsDetSpecB}
    Alpha=@n40
   )) {outputName=TerrainSpec}
  FunctionOutput(None=LinearInterpolate(
    A=ScalarParameter() {defaultValue=0.8, parameterName=Terrain-GrsDetRoughA}
    B=ScalarParameter() {defaultValue=1, parameterName=Terrain-GrsDetRoughB}
    Alpha=@n40
   )) {outputName=TerrainRough}
  LinearInterpolate(
   A=@n23.R
   B=@n10.R
   Alpha=MaterialFunctionCall(Blend Range=ScalarParameter() {parameterName=GrsDetBR}, Start Offset=ScalarParameter() {parameterName=GrsGetSO}) {materialFunction=&Distance_Blend}.Result
  )
  Saturate(None=Add(A=LinearInterpolate(Alpha=@n23.R) {constA=-0.2, constB=1.2}, B=@n20))
  TextureSample() {samplerType=SAMPLERTYPE_Normal, texture=/Game/Global/Textures/Default/Tileable_Patterns/Cloud_normals}
  TextureSample() {samplerType=SAMPLERTYPE_Normal, texture=&TilingNoise06_N}
  TextureSample() {samplerType=SAMPLERTYPE_Normal, texture=/Game/Gadgets/Assets/VinderTech_GliderChute/Glider_Vampire/Textures/T_Glider_Bat_Noises_N}
  TextureSample() {samplerType=SAMPLERTYPE_Normal, texture=/Game/Global/Textures/Default/Tileable_Patterns/2ndInfectedMapNormal}
  TextureSample() {samplerType=SAMPLERTYPE_Normal, texture=/Game/Global/Textures/Default/Noise/T_SwirlNoise_N}
  TextureSample() {samplerType=SAMPLERTYPE_Normal, texture=/Game/Global/Textures/Default/Tileable_Patterns/T_WaterSoft_N}
  TextureSample() {samplerType=SAMPLERTYPE_Normal, texture=/Game/Global/Textures/Default/Tileable_Patterns/T_StormCloudDetail01_N}
 ]}
