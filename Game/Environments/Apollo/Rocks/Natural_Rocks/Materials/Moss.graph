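# Material function that layers moss over incoming rock material attributes, colored by the shared Apollo grass masks with a Fresnel boost.
# Vertex blue, modulated by inverted green or upward normal, drives a HeightLerp against a world-aligned moss texture to blend in the moss.
# Moss adds world-aligned detail normals and full roughness.
# The difference of two phase-offset HeightLerps isolates the transition band, which MossTone tints.
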MaterialFunction /Game/Environments/Apollo/Rocks/Natural_Rocks/Materials/Moss
graph {nodes [
  n1 = FunctionInput() {inputType=FunctionInput_MaterialAttributes}
  n2 = MaterialFunctionCall() {materialFunction=/Game/Athena/Apollo/Environments/GroundFoliage/MF_ApolloGrassMasks}
  n3 = Fresnel(
   ExponentIn=ScalarParameter() {defaultValue=5, group=Moss, parameterName=Moss Fresnel Exp}
   BaseReflectFractionIn=ScalarParameter() {defaultValue=0.25, group=Moss, parameterName=Moss Fresnel Transp}
  ) {baseReflectFraction=0.5, exponent=3}
  n4 = Add(A=@n2.Result, B=Multiply(A=@n2.Result, B=@n3))
  n5 = Multiply(A=Constant() {r=0.3}, B=OneMinus(None=@n3))
  n6 = Constant() {r=1}
  n7 = TextureObject() {samplerType=SAMPLERTYPE_Normal, texture=/Game/Environments/Apollo/Sets/Terrain/Textures/T_Apollo_Moss_N}
  n8 = ScalarParameter() {defaultValue=650, group=Moss, parameterName=Moss Scale N}
  n9 = ScalarParameter() {defaultValue=0.5, group=DetailNormal, parameterName=Moss Normals Val}
  n10 = AppendVector(A=@n9, B=@n9)
  n11 = Constant()
  n12 = Add(
   A=LinearInterpolate(
    A=GetMaterialAttributes(None=@n1) {attributeGetTypes [#1]}.SubsurfaceColor
    B=Constant3Vector() {constant=<rgba 0 0 1 0>}
   )
   B=Multiply(
    A=MaterialFunctionCall(TextureObject=@n7, TextureSize=@n8) {materialFunction=&WorldAlignedNormal}.XYZ Texture
    B=AppendVector(A=@n10, B=@n11)
   )
  )
  n13 = VertexColor()
  n14 = VertexNormalWS()
  n15 = Multiply(A=@n13.B, B=Max(A=OneMinus(None=@n13.G), B=ComponentMask(None=@n14) {b=true}))
  n16 = TextureObject() {texture=/Game/Environments/Sets/Transylvania/Terrain/T_TRV_Moss_01_D}
  n17 = ScalarParameter() {defaultValue=350, group=Moss, parameterName=Moss Scale}
  n18 = StaticBool() {value=true}
  n19 = Multiply(
   A=ComponentMask(None=MaterialFunctionCall(TextureObject=@n16, TextureSize=@n17, Export Float 4=@n18) {materialFunction=&WorldAlignedTexture}.XYZ Texture) {g=true}
   B=Constant() {r=3}
  )
  n20 = ScalarParameter() {defaultValue=8, group=Moss, parameterName=MossEdgeContrast}
  n21 = GetMaterialAttributes(None=BlendMaterialAttributes(
    A=@n1
    B=MakeMaterialAttributes!(
     BaseColor=@n4
     Metallic
     Specular=@n5
     Roughness=@n6
     Anisotropy
     EmissiveColor
     Opacity
     OpacityMask
     Normal=@n12
     Tangent
     WorldPositionOffset
     SubsurfaceColor
     ClearCoat
     ClearCoatRoughness
     AmbientOcclusion
     Refraction
     CustomizedUVs_0
     CustomizedUVs_1
     CustomizedUVs_2
     CustomizedUVs_3
     CustomizedUVs_4
     CustomizedUVs_5
     CustomizedUVs_6
     CustomizedUVs_7
     PixelDepthOffset
     ShadingModel
     Displacement
    )
    Alpha=MaterialFunctionCall(Transition Phase=@n15, Height Texture=@n19, Contrast=@n20) {materialFunction=&HeightLerp}.Alpha
   ) {vertexAttributeBlendType=UseA}) {attributeGetTypes [#2]}
  n22 = MaterialFunctionCall(Transition Phase=Add(A=Constant() {r=0.0454}, B=@n15), Height Texture=@n19, Contrast=@n20) {materialFunction=&HeightLerp}
  n23 = MaterialFunctionCall(Transition Phase=Add(A=Constant() {r=-0.0359}, B=@n15), Height Texture=@n19, Contrast=@n20) {materialFunction=&HeightLerp}
  n24 = LinearInterpolate(
   B=VectorParameter() {group=Moss, parameterName=MossTone, defaultValue=<rgba 0.8281 0.7438 0.6297 0>}.RGB
   Alpha=Saturate(None=Subtract(A=@n22.Alpha, B=@n23.Alpha))
  ) {constA=1}
  FunctionOutput(None=SetMaterialAttributes!(MaterialAttributes=@n21.MaterialAttributes, BaseColor=Multiply(A=@n21.BaseColor, B=@n24)) {attributeSetTypes [#2]})
 ]}
