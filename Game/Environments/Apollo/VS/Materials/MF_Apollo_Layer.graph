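# Material function that builds a world-aligned texture layer for Apollo environment materials from diffuse, specular and normal textures with a tile size.
# WorldAlignedTextures_Complex supplies Z-only and XYZ projections, each packed into its own material attributes output.
# Diffuse RGB is tinted by BaseColorTone, the spec red channel drives both Specular and Roughness, and normals are multiplied by an input value.
# Diffuse alpha from each projection is also output as a height map.
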MaterialFunction /Game/Environments/Apollo/VS/Materials/MF_Apollo_Layer
graph {nodes [
  n1 = FunctionInput() {bUsePreviewValueAsDefault=true, inputName=Tile, inputType=FunctionInput_Scalar, previewValue=<512 0 0 1>}
  n2 = FunctionInput() {bUsePreviewValueAsDefault=true, inputName=Normal, inputType=FunctionInput_Texture2D, previewValue=<0 0 1 1>}
  n3 = MaterialFunctionCall!(
   Diffuse Texture Object=FunctionInput() {bUsePreviewValueAsDefault=true, inputName=Diffuse, inputType=FunctionInput_Texture2D}
   Spec Texture Object=FunctionInput() {bUsePreviewValueAsDefault=true, inputName=Specular, inputType=FunctionInput_Texture2D}
   TextureSize=@n1
   --------------------------
   Use seperate settings for Z projection
   Use seperate texture for Z projection
   Texture Size Z projection
   Z Plane Diffuse Texture Object
   Z Plane Spec Texture Object
   "--------------------------      "
   ProjectionTransitionContrast
   Use Texture Alpha Heightmap to Refine the Transistion
   HeightLerpTransitionContrast
   PickTheDiffuseHeightChannel
   TheDiffuseHeightChannel=Constant4Vector() {constant=<rgba 0 0 0 1>}.RGBA
   "--------------------------       "
   NormalTextureObject=@n2
   World Space Normals=StaticBool()
   "--------------------------  "
   Export Float 4 Diffuse=StaticBool() {value=true}
   Export Float 4 Spec
   "--------------------------     "
   WorldPosition
   World Space Normal
  ) {materialFunction=/Engine/Functions/Engine_MaterialFunctions02/Texturing/WorldAlignedTextures_Complex}
  n4 = FunctionInput() {bUsePreviewValueAsDefault=true, inputName=BaseColorTone, previewValue=<1 1 1 1>}
  n5 = MaterialFunctionCall(Float3=@n3.Z Spec) {materialFunction=&BreakOutFloat3Components}
  n6 = FunctionInput() {bUsePreviewValueAsDefault=true, inputName=Nomal Value, previewValue=<1 1 1 1>}
  n7 = SetMaterialAttributes!(
   MaterialAttributes
   BaseColor=Multiply(A=ComponentMask(None=@n3.Z Diffuse Texture) {b=true, g=true, r=true}, B=@n4)
   Specular=@n5.R
   Roughness=@n5.R
   Normal=Multiply(A=@n3.Normal Z Texture, B=@n6)
  ) {attributeSetTypes [
    #1
    #2
    #3
    #4
   ]}
  FunctionOutput(None=@n7) {outputName=Result Z}
  FunctionOutput(None=ComponentMask(None=@n3.Z Diffuse Texture) {a=true}) {outputName=Height map Z}
  FunctionOutput(None=ComponentMask(None=@n3.XYZ Diffuse Texture) {a=true}) {outputName=Height map XYZ}
  n8 = MaterialFunctionCall(Float3=@n3.XYZ Spec) {materialFunction=&BreakOutFloat3Components}
  n9 = SetMaterialAttributes(
   BaseColor=Multiply(A=ComponentMask(None=@n3.XYZ Diffuse Texture) {b=true, g=true, r=true}, B=@n4)
   Specular=@n8.R
   Roughness=@n8.R
   Normal=Multiply(A=@n3.Normal XYZ Texture, B=@n6)
  ) {attributeSetTypes [
    #1
    #2
    #3
    #4
   ]}
  FunctionOutput(None=@n9) {outputName=Result XYZ}
 ]}
