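# Material function that adds a wet waterline look to incoming material attributes, based on world height.
# It builds a wet variant by darkening Base Color with a scalar parameter and scaling down Specular and Roughness by constants.
# World Z relative to WaterHeight, divided by Waterline Blur and saturated, blends from the wet variant at and below the waterline to the untouched input above it.
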MaterialFunction /Game/Environments/Apollo/VS/Materials/MF_VS_Waterline
graph {nodes [
  n1 = FunctionInput() {inputType=FunctionInput_MaterialAttributes}
  n2 = GetMaterialAttributes(None=@n1) {attributeGetTypes [#1, #2, #3]}
  n3 = Multiply(A=@n2.BaseColor, B=ScalarParameter() {defaultValue=0.35, parameterName=Waterline To Diffuse}) {constB=0.5}
  n4 = Multiply(A=@n2.Specular) {constB=0.5}
  n5 = Multiply(A=@n2.Roughness, B=Constant() {r=0.25}) {constB=0.5}
  n6 = WorldPosition()
  n7 = Divide(
   A=Subtract(A=ComponentMask(None=@n6.XYZ) {b=true}, B=ScalarParameter() {defaultValue=-140.4826, parameterName=WaterHeight})
   B=ScalarParameter() {defaultValue=82.2741, parameterName=Waterline Blur}
  )
  FunctionOutput(None=BlendMaterialAttributes(
    A=SetMaterialAttributes!(MaterialAttributes=@n2.MaterialAttributes, BaseColor=@n3, Specular=@n4, Roughness=@n5) {attributeSetTypes [#1, #2, #3]}
    B=@n1
    Alpha=Saturate(None=@n7)
   ))
 ]}
