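# Apollo pebble material for instanced meshes that blends small rocks into the landscape.
# Color mixes grass-mask and dirt layers with a per-instance-random pebble tint by upward normal, while world-aligned detail textures modulate tint, specular, roughness and normals.
# After optional moss and a world-space normal conversion, height above the landscape drives Runtime Virtual Texture blending of normals and surface attributes.
# Forward and mobile paths skip that blend; all pass through a waterline function.
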Material /Game/Environments/Apollo/Rocks/Natural_Rocks/Materials/M_Pebbles_01
graph {
 properties {
  bTangentSpaceNormal=false
  bUseMaterialAttributes=true
  bUsedWithInstancedStaticMeshes=true
  pixelDepthOffsetMode=PDOM_Legacy
  usedShadingModels=Default Lit
 }
 nodes [
  n1 = ComponentMask(None=MaterialFunctionCall(
    Export Float 4=StaticBool() {value=true}
    TextureObject=TextureObjectParameter() {parameterName=Detail D, texture=/Game/Environments/Apollo/Rocks/Natural_Rocks/Textures/T_RockDetails_02D}
    TextureSize=ScalarParameter() {defaultValue=512, group=DetailNormal, parameterName=DetailNormalTile}
   ) {materialFunction=/Game/Weapons/WeaponSkins/Wraps/MaterialFunctions/WorldAlignedTexture}.XYZ Texture) {b=true, g=true, r=true}
  LinearInterpolate(A=Constant() {r=-2}, B=Constant() {r=3}, Alpha=ComponentMask(None=@n1) {r=true})
  n2 = TextureSampleParameter2D() {
   parameterName=Diffuse
   samplerSource=SSM_Wrap_WorldGroupSettings
   texture=/Game/Environments/Apollo/Rocks/Natural_Rocks/Textures/T_NaturalRock_AB_D
  }
  MaterialFunctionCall(Base=@n2.R, Blend=@n1) {materialFunction=&Blend_Overlay}
  n3 = MaterialFunctionCall() {materialFunction=/Game/Athena/Apollo/Environments/GroundFoliage/MF_ApolloGrassMasks}
  n4 = TextureObjectParameter() {
   group=Dirt
   parameterName=Specular_Dirt
   samplerType=SAMPLERTYPE_Masks
   texture=/Game/Environments/Apollo/Sets/Terrain/Textures/T_Apollo_Dirt_S_01
  }
  n5 = TextureObjectParameter() {
   group=Dirt
   parameterName=Normals_Dirt
   samplerType=SAMPLERTYPE_Normal
   texture=/Game/Environments/Apollo/Sets/Terrain/Textures/T_Apollo_Dirt_N_01
  }
  n6 = GetMaterialAttributes(None=MaterialFunctionCall!(
    Diffuse=TextureObjectParameter() {group=Dirt, parameterName=Diffuse_Dirt, texture=/Game/Environments/Apollo/Sets/Terrain/Textures/T_Apollo_Dirt_D_01}
    Specular=@n4
    Normal=@n5
    Nomal Value
    BaseColorTone
    Tile=ScalarParameter() {defaultValue=512, group=Dirt, parameterName=DirtScale}
   ) {materialFunction=/Game/Environments/Apollo/VS/Materials/MF_Apollo_Layer}.Result Z) {attributeGetTypes [#1]}
  n7 = ComponentMask(None=MaterialFunctionCall(
    TextureObject=TextureObjectParameter() {parameterName=Detail D, texture=&T_RockDetails_02D}
    TextureSize=ScalarParameter() {defaultValue=512, group=DetailNormal, parameterName=DetailNormalTile}
   ) {materialFunction=&WorldAlignedTexture}.XYZ Texture) {b=true, g=true, r=true}
  n8 = MaterialFunctionCall(Base=TextureSample(UVs=PerInstanceRandom()) {texture=/Game/Environments/Apollo/Sets/Terrain/Textures/T_Pebbles_D}.RGB, Blend=@n7) {materialFunction=&Blend_Overlay}
  n9 = ComponentMask(None=VertexNormalWS()) {b=true}
  n10 = TextureSample() {texture=/Game/Environments/Apollo/Rocks/Natural_Rocks/Textures/T_Apollo_Pebbles_S}
  n11 = ComponentMask(None=@n7) {r=true}
  n12 = Constant() {r=-2}
  n13 = Constant() {r=3}
  n14 = TextureObjectParameter() {parameterName=Detail N, samplerType=SAMPLERTYPE_Normal, texture=/Game/Environments/Apollo/Rocks/Natural_Rocks/Textures/T_RockDetails_02_N}
  n15 = MaterialFunctionCall(TextureObject=@n14, TextureSize=ScalarParameter() {defaultValue=512, group=DetailNormal, parameterName=DetailNormalTile}) {materialFunction=&WorldAlignedNormal}
  n16 = ScalarParameter() {defaultValue=0.2, group=DetailNormal, parameterName=Detail Normals Val}
  n17 = LinearInterpolate(A=@n16, B=Multiply(A=@n16) {constB=2})
  n18 = AppendVector(A=AppendVector(A=@n17, B=@n17), B=Constant())
  n19 = SetMaterialAttributes!(
   MaterialAttributes
   Base Color=LinearInterpolate(A=LinearInterpolate(A=@n3.Result, B=@n6.Base Color, Alpha=@n3.Result), B=@n8.Result, Alpha=Saturate(None=@n9))
   Specular=Multiply(A=@n8.Result, B=Multiply(A=@n10.R, B=@n11))
   Roughness=Multiply(A=LinearInterpolate(A=@n12, B=@n13, Alpha=@n11), B=@n10.B)
   Normal=Add(
    A=TextureSample() {samplerType=SAMPLERTYPE_Normal, texture=/Game/Environments/Apollo/Rocks/Natural_Rocks/Textures/T_Apollo_Pebbles_N}.RGB
    B=Multiply(A=@n15.XYZ Texture, B=@n18)
   )
  ) {attributeSetTypes [
    #1
    #2
    #3
    #4
   ]}
  n20 = MaterialFunctionCall!(Material=StaticSwitchParameter(True=MaterialFunctionCall!(In=@n19) {materialFunction=&/Moss}.Result, False=@n19) {parameterName=UseMoss}) {materialFunction=/Game/Athena/Apollo/Maps/Landscape/VT/MaterialFunctions/TransformTangentToWorld}
  n21 = RuntimeVirtualTextureSampleParameter() {materialType=BaseColor_Normal_Specular_YCoCg, parameterName=RVT, virtualTexture=/Game/Athena/Apollo/Maps/Landscape/VT/RVT_Apollo_Landscape}
  n22 = RuntimeVirtualTextureSampleParameter() {materialType=WorldHeight, parameterName=RVTHeight, virtualTexture=/Game/Athena/Apollo/Maps/Landscape/VT/RVT_Apollo_Landscape_Height}
  n23 = Subtract(A=ComponentMask(None=WorldPosition().XYZ) {b=true}, B=@n22.WorldHeight)
  n24 = Saturate(None=Divide(
    A=NamedRerouteDeclaration(None=@n23) {name=RVT Height, nodeColor=<rgba 1 0 0.5156 1>}
    B=ScalarParameter() {defaultValue=5, group=RVT, parameterName=Blend RVT Normals}
   ))
  n25 = BlendMaterialAttributes(
   A=SetMaterialAttributes!(MaterialAttributes=@n20.Result, Normal=@n21.Normal) {attributeSetTypes [#4]}
   B=@n20.Result
   Alpha=@n24
  ) {vertexAttributeBlendType=UseA}
  n26 = SetMaterialAttributes!(MaterialAttributes=@n25, Base Color=@n21.BaseColor, Specular=@n21.BaseColor, Roughness=@n21.BaseColor) {attributeSetTypes [#1, #2, #3]}
  n27 = ScalarParameter() {defaultValue=35, group=RVT, parameterName=Blend RVT}
  n28 = NamedRerouteDeclaration(None=Saturate(None=Divide(A=@n23, B=@n27))) {name=RVT Mask, nodeColor=<rgba 0 1 0.3203 1>}
  n29 = MaterialFunctionCall(
   NoRVT=@n20.Result
   RVT=MaterialProxyReplace(Realtime=BlendMaterialAttributes(A=@n26, B=@n25, Alpha=@n28) {vertexAttributeBlendType=UseA}, MaterialProxy=@n20.Result)
  ) {materialFunction=&FN_RVT_Switch_Material}
  MaterialFunctionCall!(In=ShadingPathSwitch(Default=@n29.Result, Deferred=@n29.Result, Forward=@n20.Result, Mobile=@n20.Result)) {materialFunction=/Game/Environments/Apollo/VS/Materials/MF_VS_Waterline}
  Reroute(None=@n24)
 ]
}
