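# Material function that layers tintable moss over incoming material attributes, passing them through untouched on Low quality.
# Coverage blends upward world vertex normal with upward direction from actor origin, scaled by red vertex color, then height-lerped using diffuse alpha.
# A static switch picks UV or world-aligned projection for moss textures.
# The height-lerp transition drives edge color and normal flattening, while Fresnel from the moss normal adjusts color and faint emissive.
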MaterialFunction /Game/Environments/Apollo/Rocks/Natural_Rocks/Materials/MossSoft2
graph {
 properties {description=This function is used for generating world positioned moss. This also includes parameters to increase/decrease tile size and amount of moss covering the material.}
 nodes [
  n1 = FunctionInput() {inputType=FunctionInput_MaterialAttributes}
  n2 = VertexColor()
  n3 = ComponentMask(None=VertexNormalWS()) {b=true}
  n4 = WorldPosition()
  n5 = ActorPositionWS()
  n6 = ComponentMask(None=Normalize(VectorInput=Subtract(A=@n4.XYZ, B=@n5))) {b=true}
  n7 = ScalarParameter() {defaultValue=0.5, group=Moss, parameterName=Blend2NormalDirections}
  n8 = Multiply(
   A=Add(A=LinearInterpolate(A=@n3, B=@n6, Alpha=@n7), B=ScalarParameter() {defaultValue=-0.2459, group=Moss, parameterName=Mossoffset})
   B=ScalarParameter() {defaultValue=1.2661, group=Moss, parameterName=MossAmount}
  )
  n9 = Multiply(A=TextureCoordinate(), B=ScalarParameter() {defaultValue=4, group=Moss, parameterName=Moss Size UV})
  n10 = TextureObjectParameter() {group=Moss, parameterName=Moss Diffuse, sortPriority=-2, texture=/Game/Environments/Apollo/Sets/Terrain/Textures/T_Apollo_Moss_D_01}
  n11 = ScalarParameter() {defaultValue=607.8037, group=Moss, parameterName=Moss_Texture_Size}
  n12 = ScalarParameter() {defaultValue=1.8585, group=Moss, parameterName=ProjectionContrast_Moss}
  n13 = StaticSwitchParameter(
   True=TextureSample(UVs=@n9, Tex=@n10).RGBA
   False=MaterialFunctionCall(TextureObject=@n10, TextureSize=@n11, Export Float 4=StaticBool() {value=true}, ProjectionTransitionContrast=@n12) {materialFunction=&WorldAlignedTexture}.XYZ Texture
  ) {desc=Cheap and simple, group=Moss, parameterName=UV Based Moss}
  n14 = MaterialFunctionCall(
   Transition Phase=Saturate(None=Multiply(A=@n2.R, B=@n8))
   Height Texture=ComponentMask(None=@n13) {a=true}
   Contrast=ScalarParameter() {defaultValue=3.3398, group=Moss, parameterName=MossContrastEdge}
  ) {materialFunction=&HeightLerp}
  n15 = Saturate(None=LinearInterpolate(
    A=ScalarParameter() {defaultValue=-1.1814, group=Moss, parameterName=MossColorEdge_1}
    B=ScalarParameter() {defaultValue=0.8762, group=Moss, parameterName=MossColorEdge_2}
    Alpha=@n14.Lerp Alpha No Contrast
   ))
  n16 = LinearInterpolate(
   A=VectorParameter() {group=Moss, parameterName=ColorMoss_Edge, defaultValue=<rgba 0.3595 0.4062 0.0644 1>}.RGB
   B=VectorParameter() {group=Moss, parameterName=ColorMoss, defaultValue=<rgba 0.2447 0.4688 0 1>}.RGB
   Alpha=@n15
  )
  n17 = TextureObjectParameter() {
   group=Moss
   parameterName=Moss Normal
   samplerType=SAMPLERTYPE_Normal
   sortPriority=-1
   texture=/Game/Environments/Apollo/Sets/Terrain/Textures/T_Apollo_Moss_N_01
  }
  n18 = StaticSwitchParameter(
   True=TextureSample(UVs=@n9, Tex=@n17).RGB
   False=MaterialFunctionCall(TextureObject=@n17, TextureSize=@n11, ProjectionTransitionContrast=@n12, Use High Quality Normals=StaticBool()) {materialFunction=&WorldAlignedNormal}.XYZ Texture
  ) {desc=Cheap and simple, group=Moss, parameterName=UV Based Moss}
  n19 = Fresnel(ExponentIn=ScalarParameter() {defaultValue=3.8484, group=Moss, parameterName=FresnelMosss}, Normal=Transform(None=@n18)) {baseReflectFraction=0}
  n20 = Multiply(A=@n16, B=Multiply(A=ScalarParameter() {defaultValue=-0.2, group=Moss, parameterName=ColorFresnel_Birghtness}, B=@n19))
  n21 = MaterialFunctionCall(
   Normal=@n18
   Flatness=LinearInterpolate(
    A=ScalarParameter() {defaultValue=-0.2286, group=Moss, parameterName=MossNormalFlatten_Edge}
    B=ScalarParameter() {defaultValue=0.7619, group=Moss, parameterName=MossNormalFlatten}
    Alpha=@n15
   )
  ) {materialFunction=&FlattenNormal}
  n22 = StaticBool()
  n23 = SetMaterialAttributes!(
   MaterialAttributes
   BaseColor=Multiply(A=Add(A=@n16, B=@n20), B=ComponentMask(None=@n13) {b=true, g=true, r=true})
   Metallic=ScalarParameter() {defaultValue=0.7619, group=Moss, parameterName=Metallic_Moss}
   Specular=ScalarParameter() {defaultValue=1, group=Moss, parameterName=Specular_Moss}
   Roughness=ScalarParameter() {defaultValue=0.8286, group=Moss, parameterName=Roughnes_Moss}
   EmissiveColor=Multiply(A=ScalarParameter() {defaultValue=0.01, group=Moss, parameterName=EmissiveFresnel_Birghtness}, B=@n19)
   Normal=StaticSwitch(
    True=Transform(None=@n21.Result)
    False=@n21.Result
    Value=FunctionInput(Preview=@n22) {bUsePreviewValueAsDefault=true, inputName=WorldSpaceNormals, inputType=FunctionInput_StaticBool}
   )
  ) {attributeSetTypes [
    #1
    #2
    #3
    #4
    #5
    #6
   ]}
  n24 = BlendMaterialAttributes(A=@n1, B=@n23, Alpha=@n14.Alpha)
  FunctionOutput(None=QualitySwitch(Default=@n24, Low=@n1, High=@n24, Medium=@n24))
 ]
}
